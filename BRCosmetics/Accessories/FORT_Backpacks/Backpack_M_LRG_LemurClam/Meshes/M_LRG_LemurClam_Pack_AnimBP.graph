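# Animation Blueprint for the LemurClam backpack cosmetic, posing its wings from player movement and simulating its dangling rocket strings.
# The wingPosition state machine selects pose-asset poses for ground, air, targeting, crouch/emote, sprint and skydiving, adding movement-driven lean blendspaces in some states.
# Its cached pose feeds two Anim Dynamics setups: full-strength root-relative in game, subtler world-space in the front end.
# EventGraph also computes gravity and spring values the AnimGraph never reads.
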AnimBlueprint /BRCosmetics/Accessories/FORT_Backpacks/Backpack_M_LRG_LemurClam/Meshes/M_LRG_LemurClam_Pack_AnimBP
target_skeleton=&/M_LRG_LemurClam_Pack_Skeleton
graphs [
 AnimGraph {nodes [
   Input Pose()
   n1 = "Anim Dynamics - Chain: dyn_rocket_string_01_l -> dyn_rocket_string_03_l"!(
    ComponentPose=Local To Component(LocalPose=Use cached pose 'input_pose_cache'())
    Alpha=1
    GravityOverride:Vector="0,0,-980"
    ExternalForce="0,0,0"
   ) -> (Pose) {
    bChain=true
    bUseGravityOverride=true
    simulationSpace=RootRelative
    boundBone {boneName=dyn_rocket_string_01_l}
    chainEnd {boneName=dyn_rocket_string_03_l}
    physicsBodyDefinitions [
     {
      boundBone {boneName=dyn_rocket_string_01_l}
      boxExtents=<10 10 10>
      constraintSetup {angularLimitsMax=<15 15 15>, angularLimitsMin=<-15 -15 -15>, angularTarget=<0 0 0>}
      localJointOffset=<0 0 0>
     }
     {
      boundBone {boneName=dyn_rocket_string_02_l}
      boxExtents=<10 10 10>
      constraintSetup {angularLimitsMax=<15 15 15>, angularLimitsMin=<-15 -15 -15>, angularTarget=<0 0 0>}
      localJointOffset=<0 0 0>
     }
     {
      boundBone {boneName=dyn_rocket_string_03_l}
      boxExtents=<10 10 10>
      constraintSetup {angularLimitsMax=<15 15 15>, angularLimitsMin=<-15 -15 -15>, angularTarget=<0 0 0>}
      localJointOffset=<0 0 0>
     }
    ]
    planarLimits [
     {drivingBone {boneName=pack}, planeTransform {location=<0 -5 0>, rotation=<rot 0 40 -90>, scale=<0.1 0.1 0.1>}}
     {drivingBone {boneName=pack}, planeTransform {location=<0 -6 0>, rotation=<rot 0 0 -90>, scale=<0.1 0.1 0.1>}}
     {drivingBone {boneName=pack}, planeTransform {location=<2 0 0>, rotation=<rot 0 90 -90>}}
    ]
   }
   n2 = "Anim Dynamics - Chain: dyn_rocket_string_01_r -> dyn_rocket_string_03_r"!(ComponentPose=@n1, Alpha=1, GravityOverride:Vector="0,0,-980", ExternalForce="0,0,0") -> (Pose) {
    bChain=true
    bUseGravityOverride=true
    simulationSpace=RootRelative
    boundBone {boneName=dyn_rocket_string_01_r}
    chainEnd {boneName=dyn_rocket_string_03_r}
    physicsBodyDefinitions [
     {
      boundBone {boneName=dyn_rocket_string_01_r}
      boxExtents=<10 10 10>
      constraintSetup {angularLimitsMax=<15 15 15>, angularLimitsMin=<-15 -15 -15>, angularTarget=<0 0 0>}
      localJointOffset=<0 0 0>
     }
     {
      boundBone {boneName=dyn_rocket_string_02_r}
      boxExtents=<10 10 10>
      constraintSetup {angularLimitsMax=<0 0 0>, angularLimitsMin=<0 0 0>, angularTarget=<0 0 0>}
      localJointOffset=<0 0 0>
     }
     {
      boundBone {boneName=dyn_rocket_string_03_r}
      boxExtents=<10 10 10>
      constraintSetup {angularLimitsMax=<0 0 0>, angularLimitsMin=<0 0 0>, angularTarget=<0 0 0>}
      localJointOffset=<0 0 0>
     }
    ]
    planarLimits [
     {drivingBone {boneName=pack}, planeTransform {location=<0 -5 0>, rotation=<rot 0 -40 -90>, scale=<0.1 0.1 0.1>}}
     {drivingBone {boneName=pack}, planeTransform {location=<0 -6 0>, rotation=<rot 0 0 -90>, scale=<0.1 0.1 0.1>}}
     {drivingBone {boneName=pack}, planeTransform {location=<-2 0 0>, rotation=<rot 0 90 90>}}
    ]
   }
   n3 = "Anim Dynamics - Chain: dyn_rocket_string_01_l -> dyn_rocket_string_03_l"(
    ComponentPose=Local To Component(LocalPose=Use cached pose 'input_pose_cache'())
    Alpha=0.25
    GravityOverride="0,0,-980"
    ExternalForce="0,0,0"
   ) {
    bChain=true
    bUseGravityOverride=true
    simulationSpace=World
    boundBone {boneName=dyn_rocket_string_01_l}
    chainEnd {boneName=dyn_rocket_string_03_l}
    physicsBodyDefinitions [
     {
      boundBone {boneName=dyn_rocket_string_01_l}
      boxExtents=<10 10 10>
      constraintSetup {angularLimitsMax=<5 5 5>, angularLimitsMin=<-5 -5 -5>, angularTarget=<0 0 0>}
      localJointOffset=<0 0 0>
     }
     {
      boundBone {boneName=dyn_rocket_string_02_l}
      boxExtents=<10 10 10>
      constraintSetup {angularLimitsMax=<5 5 5>, angularLimitsMin=<-5 -5 -5>, angularTarget=<0 0 0>}
      localJointOffset=<0 0 0>
     }
     {
      boundBone {boneName=dyn_rocket_string_03_l}
      boxExtents=<10 10 10>
      constraintSetup {angularLimitsMax=<5 5 5>, angularLimitsMin=<-5 -5 -5>, angularTarget=<0 0 0>}
      localJointOffset=<0 0 0>
     }
    ]
    planarLimits [{drivingBone {boneName=pack}, planeTransform {location=<0 -2 0>, rotation=<rot 0 0 -90>, scale=<0.1 0.1 0.1>}}]
   }
   n4 = "Anim Dynamics - Chain: dyn_rocket_string_01_r -> dyn_rocket_string_03_r"(ComponentPose=@n3, Alpha=0.25, GravityOverride="0,0,-980", ExternalForce="0,0,0") {
    bChain=true
    bUseGravityOverride=true
    simulationSpace=World
    boundBone {boneName=dyn_rocket_string_01_r}
    chainEnd {boneName=dyn_rocket_string_03_r}
    physicsBodyDefinitions [
     {
      boundBone {boneName=dyn_rocket_string_01_r}
      boxExtents=<10 10 10>
      constraintSetup {angularLimitsMax=<5 5 5>, angularLimitsMin=<-5 -5 -5>, angularTarget=<0 0 0>}
      localJointOffset=<0 0 0>
     }
     {
      boundBone {boneName=dyn_rocket_string_02_r}
      boxExtents=<10 10 10>
      constraintSetup {angularLimitsMax=<5 5 5>, angularLimitsMin=<-5 -5 -5>, angularTarget=<0 0 0>}
      localJointOffset=<0 0 0>
     }
     {
      boundBone {boneName=dyn_rocket_string_03_r}
      boxExtents=<10 10 10>
      constraintSetup {angularLimitsMax=<5 5 5>, angularLimitsMin=<-5 -5 -5>, angularTarget=<0 0 0>}
      localJointOffset=<0 0 0>
     }
    ]
    planarLimits [{drivingBone {boneName=pack}, planeTransform {location=<0 -2 0>, rotation=<rot 0 0 -90>, scale=<0.1 0.1 0.1>}}]
   }
   Output Pose(Result=Blend Poses by bool(
     BlendPose_0=Component To Local(ComponentPose=@n4)
     BlendPose_1=Component To Local(ComponentPose=@n2)
     bActiveValue=Get use_FE_float_logic() -> (use_FE_float_logic:bool)
    )) {layerGroup=None}
   AnimGraphNode_StateMachine_0 = wingPosition() {is=title AnimGraphNode_StateMachine}
   Save cached pose 'input_pose_cache'(Pose=@AnimGraphNode_StateMachine_0) {cacheName=input_pose_cache}
  ]}
 wingPosition {
  entry=Base
  path=AnimGraph.AnimGraphNode_StateMachine_0.wingPosition
  ids=id
  nodes [
   AnimStateNode_0 = Base() {is=title AnimStateNode}
   AnimStateNode_1 = Jumping() {is=title AnimStateNode}
   AnimStateNode_2 = Falling() {is=title AnimStateNode}
   AnimStateNode_3 = Targeting() {is=title AnimStateNode}
   AnimStateNode_5 = SkydiveMode() {is=title AnimStateNode}
   AnimStateNode_6 = Crouching() {is=title AnimStateNode}
   AnimStateNode_9 = Tactical Sprint() {is=title AnimStateNode}
   AnimStateTransitionNode_0 = AnimStateTransitionNode() {
    blendMode=Custom
    crossfadeDuration=0.5
    customBlendCurve=/Game/Accessories/FORT_Backpacks/M_MED_Demon_Pack/Animation/JumpTransitionBlendCurve
    priorityOrder=2
   }
   AnimStateTransitionNode_1 = AnimStateTransitionNode() {
    blendMode=Custom
    crossfadeDuration=0.5
    customBlendCurve=/Game/Accessories/FORT_Backpacks/M_MED_Demon_Pack/Animation/JumpTransitionBlendOutCurve
   }
   AnimStateTransitionNode_10 = AnimStateTransitionNode()
   AnimStateTransitionNode_11 = AnimStateTransitionNode()
   AnimStateTransitionNode_14 = AnimStateTransitionNode()
   AnimStateTransitionNode_15 = AnimStateTransitionNode()
   AnimStateTransitionNode_16 = AnimStateTransitionNode()
   AnimStateTransitionNode_18 = AnimStateTransitionNode()
   AnimStateTransitionNode_2 = AnimStateTransitionNode() {crossfadeDuration=0.3}
   AnimStateTransitionNode_3 = AnimStateTransitionNode() {
    blendMode=Custom
    crossfadeDuration=0.3
    customBlendCurve=&JumpTransitionBlendOutCurve
   }
   AnimStateTransitionNode_4 = AnimStateTransitionNode() {
    blendMode=Custom
    crossfadeDuration=0.3
    customBlendCurve=&JumpTransitionBlendCurve
   }
   AnimStateTransitionNode_5 = AnimStateTransitionNode()
   AnimStateTransitionNode_6 = AnimStateTransitionNode() {crossfadeDuration=0.25}
   AnimStateTransitionNode_7 = AnimStateTransitionNode() {crossfadeDuration=0.25}
   AnimStateTransitionNode_8 = AnimStateTransitionNode()
   AnimStateTransitionNode_9 = AnimStateTransitionNode()
  ]
  links [
   {
    dst=Crouching
    rule=AnimGraph.AnimGraphNode_StateMachine_0.wingPosition.AnimStateTransitionNode_16.AnimationTransitionGraph_0
    src=Base
    via=AnimStateTransitionNode_16
   }
   {
    dst=Falling
    rule=AnimGraph.AnimGraphNode_StateMachine_0.wingPosition.AnimStateTransitionNode_4.Transition
    src=Base
    via=AnimStateTransitionNode_4
   }
   {
    dst=Jumping
    rule=AnimGraph.AnimGraphNode_StateMachine_0.wingPosition.AnimStateTransitionNode_0.Transition
    src=Base
    via=AnimStateTransitionNode_0
   }
   {
    dst=SkydiveMode
    rule=AnimGraph.AnimGraphNode_StateMachine_0.wingPosition.AnimStateTransitionNode_15.Transition
    src=Base
    via=AnimStateTransitionNode_15
   }
   {
    dst=Tactical Sprint
    rule=AnimGraph.AnimGraphNode_StateMachine_0.wingPosition.AnimStateTransitionNode_8.Transition
    src=Base
    via=AnimStateTransitionNode_8
   }
   {
    dst=Targeting
    rule=AnimGraph.AnimGraphNode_StateMachine_0.wingPosition.AnimStateTransitionNode_6.Transition
    src=Base
    via=AnimStateTransitionNode_6
   }
   {
    dst=Base
    rule=AnimGraph.AnimGraphNode_StateMachine_0.wingPosition.AnimStateTransitionNode_18.AnimationTransitionGraph_1
    src=Crouching
    via=AnimStateTransitionNode_18
   }
   {
    dst=Base
    rule=AnimGraph.AnimGraphNode_StateMachine_0.wingPosition.AnimStateTransitionNode_3.Transition
    src=Falling
    via=AnimStateTransitionNode_3
   }
   {
    dst=Base
    rule=AnimGraph.AnimGraphNode_StateMachine_0.wingPosition.AnimStateTransitionNode_1.Transition
    src=Jumping
    via=AnimStateTransitionNode_1
   }
   {
    dst=Falling
    rule=AnimGraph.AnimGraphNode_StateMachine_0.wingPosition.AnimStateTransitionNode_2.Transition
    src=Jumping
    via=AnimStateTransitionNode_2
   }
   {
    dst=SkydiveMode
    rule=AnimGraph.AnimGraphNode_StateMachine_0.wingPosition.AnimStateTransitionNode_5.Transition
    src=Jumping
    via=AnimStateTransitionNode_5
   }
   {
    dst=Tactical Sprint
    rule=AnimGraph.AnimGraphNode_StateMachine_0.wingPosition.AnimStateTransitionNode_14.Transition
    src=Jumping
    via=AnimStateTransitionNode_14
   }
   {
    dst=Base
    rule=AnimGraph.AnimGraphNode_StateMachine_0.wingPosition.AnimStateTransitionNode_11.Transition
    src=SkydiveMode
    via=AnimStateTransitionNode_11
   }
   {
    dst=Base
    rule=AnimGraph.AnimGraphNode_StateMachine_0.wingPosition.AnimStateTransitionNode_9.Transition
    src=Tactical Sprint
    via=AnimStateTransitionNode_9
   }
   {
    dst=Jumping
    rule=AnimGraph.AnimGraphNode_StateMachine_0.wingPosition.AnimStateTransitionNode_10.Transition
    src=Tactical Sprint
    via=AnimStateTransitionNode_10
   }
   {
    dst=Base
    rule=AnimGraph.AnimGraphNode_StateMachine_0.wingPosition.AnimStateTransitionNode_7.Transition
    src=Targeting
    via=AnimStateTransitionNode_7
   }
  ]
 }
 Base {
  path=AnimGraph.AnimGraphNode_StateMachine_0.wingPosition.AnimStateNode_0.Base
  nodes [
   n1 = Get PlayerVector_FwdBwd() -> (PlayerVector_FwdBwd:double)
   n2 = Get PlayerVector_LR() -> (PlayerVector_LR:double)
   n3 = LemurClam_Wings_Movement_Poses_PoseAsset : Base!(PoseWeight:float=1) -> (Pose:Pose Link Structure) {
    is=title AnimGraphNode_PoseByName
    poseAsset=/BRCosmetics/Accessories/FORT_Backpacks/Backpack_M_LRG_LemurClam/Animation/LemurClam_Wings_Movement_Poses_PoseAsset
    poseName=Base
   }
   Output Animation Pose(Result=Apply Additive(
     Base=@n3
     Additive=Blendspace Player 'BS_Wing_Leans_Ground'(X=@n2, Y=@n1) {blendSpace=/BRCosmetics/Accessories/FORT_Backpacks/Backpack_M_LRG_LemurClam/Animation/BS_Wing_Leans_Ground}
    ) {alphaBoolBlend {blendOption=Linear}})
  ]
 }
 Jumping {
  path=AnimGraph.AnimGraphNode_StateMachine_0.wingPosition.AnimStateNode_1.Jumping
  nodes [
   n1 = Get PlayerVector_FwdBwd()
   n2 = Get PlayerVector_LR()
   n3 = LemurClam_Wings_Movement_Poses_PoseAsset : Jump!(PoseWeight:float=0.5) -> (Pose:Pose Link Structure) {
    is=title AnimGraphNode_PoseByName
    poseAsset=&LemurClam_Wings_Movement_Poses_PoseAsset
    poseName=Jump
   }
   Output Animation Pose(Result=Apply Additive(
     Base=@n3
     Additive=Blendspace Player 'BS_Wing_Leans_Air'(X=@n2, Y=@n1) {blendSpace=/BRCosmetics/Accessories/FORT_Backpacks/Backpack_M_LRG_LemurClam/Animation/BS_Wing_Leans_Air}
    ) {alphaBoolBlend {blendOption=Linear}})
  ]
 }
 Falling {
  path=AnimGraph.AnimGraphNode_StateMachine_0.wingPosition.AnimStateNode_2.Falling
  nodes [
   n1 = Get PlayerVector_FwdBwd()
   n2 = Get PlayerVector_LR()
   n3 = LemurClam_Wings_Movement_Poses_PoseAsset : Jump(PoseWeight=0.5) {poseAsset=&LemurClam_Wings_Movement_Poses_PoseAsset, poseName=Jump}
   Output Animation Pose(Result=Apply Additive(
     Base=@n3
     Additive=Blendspace Player 'BS_Wing_Leans_Air'(X=@n2, Y=@n1) {blendSpace=&BS_Wing_Leans_Air}
    ) {alphaBoolBlend {blendOption=Linear}})
  ]
 }
 Targeting {
  path=AnimGraph.AnimGraphNode_StateMachine_0.wingPosition.AnimStateNode_3.Targeting
  nodes [
   n1 = LemurClam_Wings_Movement_Poses_PoseAsset : Targeting!(PoseWeight:float=0.6) -> (Pose:Pose Link Structure) {
    is=title AnimGraphNode_PoseByName
    poseAsset=&LemurClam_Wings_Movement_Poses_PoseAsset
    poseName=Targeting
   }
   Output Animation Pose(Result=@n1)
  ]
 }
 SkydiveMode {
  path=AnimGraph.AnimGraphNode_StateMachine_0.wingPosition.AnimStateNode_5.SkydiveMode
  nodes [
   AnimGraphNode_StateMachine_0 = Skydiving State() {is=title AnimGraphNode_StateMachine}
   Output Animation Pose(Result=@AnimGraphNode_StateMachine_0)
  ]
 }
 Skydiving State {
  entry=Parachuting
  path=AnimGraph.AnimGraphNode_StateMachine_0.wingPosition.AnimStateNode_5.SkydiveMode.AnimGraphNode_StateMachine_0.Skydiving State
  ids=id
  nodes [
   AnimStateNode_1 = Diving() {is=title AnimStateNode}
   AnimStateNode_2 = Floating() {is=title AnimStateNode}
   AnimStateNode_4 = Parachuting() {is=title AnimStateNode}
   AnimStateTransitionNode_1 = AnimStateTransitionNode() {crossfadeDuration=0.5}
   AnimStateTransitionNode_10 = AnimStateTransitionNode()
   AnimStateTransitionNode_12 = AnimStateTransitionNode()
   AnimStateTransitionNode_13 = AnimStateTransitionNode()
   AnimStateTransitionNode_3 = AnimStateTransitionNode()
   AnimStateTransitionNode_5 = AnimStateTransitionNode() {crossfadeDuration=0.5}
  ]
  links [
   {
    dst=Floating
    rule=AnimGraph.AnimGraphNode_StateMachine_0.wingPosition.AnimStateNode_5.SkydiveMode.AnimGraphNode_StateMachine_0.Skydiving State.AnimStateTransitionNode_1.AnimationTransitionGraph_0
    src=Diving
    via=AnimStateTransitionNode_1
   }
   {
    dst=Parachuting
    rule=AnimGraph.AnimGraphNode_StateMachine_0.wingPosition.AnimStateNode_5.SkydiveMode.AnimGraphNode_StateMachine_0.Skydiving State.AnimStateTransitionNode_10.AnimationTransitionGraph_1
    src=Diving
    via=AnimStateTransitionNode_10
   }
   {
    dst=Diving
    rule=AnimGraph.AnimGraphNode_StateMachine_0.wingPosition.AnimStateNode_5.SkydiveMode.AnimGraphNode_StateMachine_0.Skydiving State.AnimStateTransitionNode_5.AnimationTransitionGraph_0
    src=Floating
    via=AnimStateTransitionNode_5
   }
   {
    dst=Parachuting
    rule=AnimGraph.AnimGraphNode_StateMachine_0.wingPosition.AnimStateNode_5.SkydiveMode.AnimGraphNode_StateMachine_0.Skydiving State.AnimStateTransitionNode_3.AnimationTransitionGraph_1
    src=Floating
    via=AnimStateTransitionNode_3
   }
   {
    dst=Diving
    rule=AnimGraph.AnimGraphNode_StateMachine_0.wingPosition.AnimStateNode_5.SkydiveMode.AnimGraphNode_StateMachine_0.Skydiving State.AnimStateTransitionNode_13.AnimationTransitionGraph_1
    src=Parachuting
    via=AnimStateTransitionNode_13
   }
   {
    dst=Floating
    rule=AnimGraph.AnimGraphNode_StateMachine_0.wingPosition.AnimStateNode_5.SkydiveMode.AnimGraphNode_StateMachine_0.Skydiving State.AnimStateTransitionNode_12.AnimationTransitionGraph_1
    src=Parachuting
    via=AnimStateTransitionNode_12
   }
  ]
 }
 Diving {
  path=AnimGraph.AnimGraphNode_StateMachine_0.wingPosition.AnimStateNode_5.SkydiveMode.AnimGraphNode_StateMachine_0.Skydiving State.AnimStateNode_1.Diving
  nodes [
   n1 = Get PlayerVector_FwdBwd()
   n2 = Get PlayerVector_LR()
   n3 = LemurClam_Wings_Movement_Poses_PoseAsset : Dive!(PoseWeight:float=0.1) -> (Pose:Pose Link Structure) {
    is=title AnimGraphNode_PoseByName
    poseAsset=&LemurClam_Wings_Movement_Poses_PoseAsset
    poseName=Dive
   }
   Output Animation Pose(Result=Apply Additive(
     Base=@n3
     Additive=Blendspace Player 'BS_Wing_Leans_Air'(X=@n2, Y=@n1) {blendSpace=&BS_Wing_Leans_Air}
    ) {alphaBoolBlend {blendOption=Linear}})
  ]
 }
 Floating {
  path=AnimGraph.AnimGraphNode_StateMachine_0.wingPosition.AnimStateNode_5.SkydiveMode.AnimGraphNode_StateMachine_0.Skydiving State.AnimStateNode_2.Floating
  nodes [
   n1 = Get PlayerVector_FwdBwd()
   n2 = Get PlayerVector_LR()
   n3 = LemurClam_Wings_Movement_Poses_PoseAsset : Float!(PoseWeight:float=1) -> (Pose:Pose Link Structure) {
    is=title AnimGraphNode_PoseByName
    poseAsset=&LemurClam_Wings_Movement_Poses_PoseAsset
    poseName=Float
   }
   Output Animation Pose(Result=Apply Additive(
     Base=@n3
     Additive=Blendspace Player 'BS_Wing_Leans_Air'(X=@n2, Y=@n1) {blendSpace=&BS_Wing_Leans_Air}
    ) {alphaBoolBlend {blendOption=Linear}})
  ]
 }
 Parachuting {
  path=AnimGraph.AnimGraphNode_StateMachine_0.wingPosition.AnimStateNode_5.SkydiveMode.AnimGraphNode_StateMachine_0.Skydiving State.AnimStateNode_4.Parachuting
  nodes [
   n1 = Get PlayerVector_FwdBwd()
   n2 = Get PlayerVector_LR()
   n3 = LemurClam_Wings_Movement_Poses_PoseAsset : Float(PoseWeight=1) {
    poseAsset=&LemurClam_Wings_Movement_Poses_PoseAsset
    poseName=Float
   }
   Output Animation Pose(Result=Apply Additive(
     Base=@n3
     Additive=Blendspace Player 'BS_Wing_Leans_Air'(X=@n2, Y=@n1) {blendSpace=&BS_Wing_Leans_Air}
    ) {alphaBoolBlend {blendOption=Linear}})
  ]
 }
 AnimationTransitionGraph_0 {
  path=AnimGraph.AnimGraphNode_StateMachine_0.wingPosition.AnimStateNode_5.SkydiveMode.AnimGraphNode_StateMachine_0.Skydiving State.AnimStateTransitionNode_1.AnimationTransitionGraph_0
  nodes [Result(bCanEnterTransition=NOT Boolean(A=Get bIsSkydiveDiveMode()))]
 }
 AnimationTransitionGraph_1 {
  path=AnimGraph.AnimGraphNode_StateMachine_0.wingPosition.AnimStateNode_5.SkydiveMode.AnimGraphNode_StateMachine_0.Skydiving State.AnimStateTransitionNode_10.AnimationTransitionGraph_1
  nodes [Result(bCanEnterTransition=Get bIsParachuteOpened())]
 }
 AnimationTransitionGraph_1 {
  path=AnimGraph.AnimGraphNode_StateMachine_0.wingPosition.AnimStateNode_5.SkydiveMode.AnimGraphNode_StateMachine_0.Skydiving State.AnimStateTransitionNode_12.AnimationTransitionGraph_1
  nodes [Result(bCanEnterTransition=NOT Boolean(A=Get bIsParachuteOpened()))]
 }
 AnimationTransitionGraph_1 {
  path=AnimGraph.AnimGraphNode_StateMachine_0.wingPosition.AnimStateNode_5.SkydiveMode.AnimGraphNode_StateMachine_0.Skydiving State.AnimStateTransitionNode_13.AnimationTransitionGraph_1
  nodes [Result(bCanEnterTransition=NOT Boolean(A=Get bIsParachuteOpened()))]
 }
 AnimationTransitionGraph_1 {
  path=AnimGraph.AnimGraphNode_StateMachine_0.wingPosition.AnimStateNode_5.SkydiveMode.AnimGraphNode_StateMachine_0.Skydiving State.AnimStateTransitionNode_3.AnimationTransitionGraph_1
  nodes [Result(bCanEnterTransition=Get bIsParachuteOpened())]
 }
 AnimationTransitionGraph_0 {
  path=AnimGraph.AnimGraphNode_StateMachine_0.wingPosition.AnimStateNode_5.SkydiveMode.AnimGraphNode_StateMachine_0.Skydiving State.AnimStateTransitionNode_5.AnimationTransitionGraph_0
  nodes [Result(bCanEnterTransition=Get bIsSkydiveDiveMode())]
 }
 Crouching {
  path=AnimGraph.AnimGraphNode_StateMachine_0.wingPosition.AnimStateNode_6.Crouching
  nodes [
   n1 = LemurClam_Wings_Movement_Poses_PoseAsset : Crouch!(PoseWeight:float=0.3) -> (Pose:Pose Link Structure) {
    is=title AnimGraphNode_PoseByName
    poseAsset=&LemurClam_Wings_Movement_Poses_PoseAsset
    poseName=Crouch
   }
   Output Animation Pose(Result=@n1)
  ]
 }
 Tactical Sprint {
  path=AnimGraph.AnimGraphNode_StateMachine_0.wingPosition.AnimStateNode_9.Tactical Sprint
  nodes [
   n1 = LemurClam_Wings_Movement_Poses_PoseAsset : Sprint!(PoseWeight:float=0.5) -> (Pose:Pose Link Structure) {
    is=title AnimGraphNode_PoseByName
    poseAsset=&LemurClam_Wings_Movement_Poses_PoseAsset
    poseName=Sprint
   }
   Output Animation Pose(Result=@n1)
  ]
 }
 Transition {
  path=AnimGraph.AnimGraphNode_StateMachine_0.wingPosition.AnimStateTransitionNode_0.Transition
  nodes [Result(bCanEnterTransition=Get bIsJumping())]
 }
 Transition {
  path=AnimGraph.AnimGraphNode_StateMachine_0.wingPosition.AnimStateTransitionNode_1.Transition
  nodes [
   Result(bCanEnterTransition=NOT Boolean(A=Get bIsJumping()))
   Get bShouldPredictLanding() -> (bShouldPredictLanding:bool)
   Get bLandingPredicted() -> (bLandingPredicted:bool)
   Get bDebugLandPrediction() -> (bDebugLandPrediction:bool)
   Get bIsFalling()
  ]
 }
 Transition {
  path=AnimGraph.AnimGraphNode_StateMachine_0.wingPosition.AnimStateTransitionNode_10.Transition
  nodes [Result(bCanEnterTransition=Get bIsJumping())]
 }
 Transition {
  path=AnimGraph.AnimGraphNode_StateMachine_0.wingPosition.AnimStateTransitionNode_11.Transition
  nodes [Result(bCanEnterTransition=NOT Boolean(A=Get bIsSkyDiving()))]
 }
 Transition {
  path=AnimGraph.AnimGraphNode_StateMachine_0.wingPosition.AnimStateTransitionNode_14.Transition
  nodes [Result(bCanEnterTransition=NOT Boolean(A=Get bIsJumping()))]
 }
 Transition {
  path=AnimGraph.AnimGraphNode_StateMachine_0.wingPosition.AnimStateTransitionNode_15.Transition
  nodes [Result(bCanEnterTransition=Get bIsSkyDiving())]
 }
 AnimationTransitionGraph_0 {
  path=AnimGraph.AnimGraphNode_StateMachine_0.wingPosition.AnimStateTransitionNode_16.AnimationTransitionGraph_0
  nodes [
   Result(bCanEnterTransition=Get Is Crouching or Emoting() -> (Is Crouching or Emoting:bool))
   OR Boolean(A=Get bIsCrouching(), B=Get bIsPlayingEmote())
  ]
 }
 AnimationTransitionGraph_1 {
  path=AnimGraph.AnimGraphNode_StateMachine_0.wingPosition.AnimStateTransitionNode_18.AnimationTransitionGraph_1
  nodes [
   Result(bCanEnterTransition=NOT Boolean(A=Get Is Crouching or Emoting()))
   OR Boolean(A=Get bIsCrouching(), B=Get bIsPlayingEmote())
  ]
 }
 Transition {
  path=AnimGraph.AnimGraphNode_StateMachine_0.wingPosition.AnimStateTransitionNode_2.Transition
  nodes [Result(bCanEnterTransition=Get bIsFalling())]
 }
 Transition {
  path=AnimGraph.AnimGraphNode_StateMachine_0.wingPosition.AnimStateTransitionNode_3.Transition
  nodes [Result(bCanEnterTransition=NOT Boolean(A=Get bIsFalling()))]
 }
 Transition {
  path=AnimGraph.AnimGraphNode_StateMachine_0.wingPosition.AnimStateTransitionNode_4.Transition
  nodes [Result(bCanEnterTransition=Get bIsFalling())]
 }
 Transition {
  path=AnimGraph.AnimGraphNode_StateMachine_0.wingPosition.AnimStateTransitionNode_5.Transition
  nodes [Result(bCanEnterTransition=Get bIsSkyDiving())]
 }
 Transition {
  path=AnimGraph.AnimGraphNode_StateMachine_0.wingPosition.AnimStateTransitionNode_6.Transition
  nodes [Result(bCanEnterTransition=Get bIsTargeting())]
 }
 Transition {
  path=AnimGraph.AnimGraphNode_StateMachine_0.wingPosition.AnimStateTransitionNode_7.Transition
  nodes [Result(bCanEnterTransition=NOT Boolean(A=Get bIsTargeting()))]
 }
 Transition {
  path=AnimGraph.AnimGraphNode_StateMachine_0.wingPosition.AnimStateTransitionNode_8.Transition
  nodes [Result(bCanEnterTransition=Get bIsTacticalSprinting())]
 }
 Transition {
  path=AnimGraph.AnimGraphNode_StateMachine_0.wingPosition.AnimStateTransitionNode_9.Transition
  nodes [Result(bCanEnterTransition=NOT Boolean(A=Get bIsTacticalSprinting()))]
 }
 EventGraph {nodes [
   Event BlueprintUpdateAnimation()
   > n1 = Sequence!(execute) -> (then_0, then_1, then_2:Exec, then_3:Exec, then_4:Exec, then_5:Exec, then_6:Exec)
   > Set use_FE_float_logic(use_FE_float_logic:bool=Get bIsInFrontEnd()) -> (Output_Get:bool)
   n2 = PlayerMovement(execute=@n1.then_1, AnimInstance=Self-Reference(), XDotProductCoefficient=1, YDotProductCoefficient=1, ZDotProductCoefficient=1)
   > Set PlayerVector_LR(PlayerVector_LR:double=@n2) -> (Output_Get:double)
   > Set PlayerVector_FwdBwd(PlayerVector_FwdBwd:double=@n2.YMovement) -> (Output_Get:double)
   n3 = NOT Boolean(A=Get bIsSkyDiving())
   n4 = NOT Boolean(A=Get bIsOnGround() -> (bIsOnGround:bool))
   > Set JumpingForwardAlpha(JumpingForwardAlpha:double=float * float(A:bool=AND Boolean(A=@n4, B=@n3), B=Get PlayerVector_FwdBwd()) {is=title K2Node_PromotableOperator}) -> (Output_Get:double)
   n5 = NOT Boolean(A=Get bIsSkyDiving())
   n6 = NOT Boolean(A=Get bIsFalling())
   n7 = NOT Boolean(A=Get bIsOnGround())
   > Set JumpingMovement(JumpingMovement:bool=AND Boolean(A=Get bIsMoving2D(), B=AND Boolean!(A=@n6, B=@n5, C:bool=@n7) -> (ReturnValue))) -> (Output_Get:bool)
   n8 = Get bIsSliding()
   > Set CrouchStrafe(CrouchStrafe:bool=AND Boolean(A=Get bIsCrouching(), B=Get bIsMoving2D(), C=NOT Boolean(A=@n8))) -> (Output_Get:bool)
   n9 = Get bIsCrouching()
   > Set IsMovingStanding(IsMovingStanding:bool=AND Boolean(A=Get bIsMoving2D(), B=Get bIsOnGround(), C=NOT Boolean(A=@n9))) -> (Output_Get:bool)
   n10 = Get bIsSkyDiving()
   n11 = Get PlayerVector_LR()
   n12 = Get bIsFalling()
   n13 = Get bIsJumping()
   > Set Jumping RL(Jumping RL:double=float * float!(A:double="Absolute (Float)"(A=@n11), B:bool=OR Boolean(A=@n12, B=@n13), C:bool=NOT Boolean(A=@n10)) -> (ReturnValue)) -> (Output_Get:double)
   n14 = Get bIsSkyDiving()
   n15 = Get bIsJumping()
   n16 = Get bIsFalling()
   > Set JumpingFalling(JumpingFalling:bool=AND Boolean!(A=OR Boolean(A=@n15, B=@n16), B=NOT Boolean(A=@n14)) -> (ReturnValue)) -> (Output_Get:bool)
   n17 = Get PlayerVector_LR()
   > Set RightMovement(RightMovement:double=float * float!(A=@n17, B=Get IsMovingStanding() -> (IsMovingStanding:bool)) -> (ReturnValue)) -> (Output_Get:double)
   > Set LeftMovement(LeftMovement:double=float * float(A=@n17, B:double=-1)) -> (Output_Get:double)
   > Set ForwardGroundMovement(ForwardGroundMovement:double=float * float(A=Get PlayerVector_FwdBwd(), B:bool=Get bIsOnGround())) -> (Output_Get:double)
   > Set Is Crouching or Emoting(Is Crouching or Emoting:bool=OR Boolean(A=Get bIsCrouching(), B=Get bIsPlayingEmote())) -> (Output_Get:bool)
   n18 = Branch(execute=@n1.then_3, Condition=Get bIsSkyDiving())
   > n19 = Branch(Condition="float > float"(A=Get PawnVelocityZ(), B=50) {is=title K2Node_CallFunction})
   > Set BwdMovement(BwdMovement:double=float * float(A=Get BwdMovement() -> (BwdMovement:double), B=0) {is=title K2Node_CommutativeAssociativeBinaryOperator}) -> (Output_Get:double)
   Set BwdMovement(execute=[@n18.else, @n19.else], BwdMovement=float * float(A=Get BwdMovement(), B=1))
   n20 = PlayerMovement(execute=@n1.then_4, AnimInstance=Self-Reference(), XDotProductCoefficient=1, YDotProductCoefficient=1, ZDotProductCoefficient=1)
   > n21 = Branch(Condition=Get bIsDBNOCarried())
   > Set BwdMovement(BwdMovement=0.5)
   n22 = Branch(execute=@n21.else, Condition=AND Boolean(A=Get bIsTargeting(), B=Get bIsCrouching()))
   n23 = MapRangeClamped(
    Value=float + float(A="Absolute (Float)"(A=@n20), B=MapRangeClamped(Value=@n20.YMovement, InRangeA=-1, InRangeB=1, OutRangeA=0.3)) {is=title K2Node_CommutativeAssociativeBinaryOperator}
    InRangeA=0
    InRangeB=1.65
   )
   > Set BwdMovement(BwdMovement=float * float(A=@n23, B:double=0.25) {is=title K2Node_PromotableOperator})
   Set BwdMovement(execute=@n22.else, BwdMovement=float * float(A=@n23, B=0.8))
   n24 = "Gravity Override State Function (Nativized)"!(
    execute=@n1.then_5
    CharacterPartAnimInstance:Custom Character Part Anim Instance Object Reference=Self-Reference()
    GravityOverrideParameterSet:Fort Gravity Override Parameter Set Structure=Get Gravity Override Params_0() -> (Gravity Override Params_0:Fort Gravity Override Parameter Set Structure)
    RigidBodyAnimNodeParameterSet:Fort Rigid Body Anim Node Parameter Set Structure
    DebugDraw:bool=false
    GameTimeInSeconds:float=GetGameTimeInSeconds()
    DebugJointName:Name=root
    StateLogicLayerName:Name=none
   ) -> (
    then
    OutGravityOverride:Vector
    OutCurrentState:Name=None
    OutGravityOverrideParameters:Fort Gravity Override Parameters Structure
    OutRigidBodyAnimNodeParameters:Fort Rigid Body Anim Node Parameters Structure
   )
   > Set GravityOverride(GravityOverride=@n24)
   > n25 = Branch(Condition=Get bIsInFrontEnd())
   > Set GravityOverride(GravityOverride="0,0,-980")
   Set GravityOverride(execute=@n25.else, GravityOverride="0,0,-980")
   > n26 = Branch(Condition=Get bIsSkydiveFloating())
   > Set GravityOverride(GravityOverride="0,0,100")
   Set GravityOverride(execute=@n26.else, GravityOverride="0,0,-980")
   > n27 = Branch(Condition=Get bIsSkydiveDiveMode())
   > Set GravityOverride(GravityOverride="0,0,100")
   Set GravityOverride(execute=@n27.else, GravityOverride="0,0,-980")
   n28 = NOT Boolean(A=Get bIsCrouching())
   n29 = NOT Boolean(A=Get bIsMoving2D())
   n30 = Get bIsOnGround()
   Set StandingIdle(
    execute=@n1.then_6
    StandingIdle:bool=OR Boolean(A=AND Boolean!(A=@n30, B=@n28, C:bool=@n29, D:bool) -> (ReturnValue), B=Get bIsInFrontEnd())
   ) -> (Output_Get:bool)
   > n31 = Branch(Condition=Get bIsInFrontEnd())
   > Set Angular Spring Constant R(Angular Spring Constant R:double=12000) -> (Output_Get:double)
   > Set Angular Spring Constant(Angular Spring Constant:double=9000) -> (Output_Get:double)
   Set Angular Spring Constant R(execute=@n31.else, Angular Spring Constant R=15000)
   > Set Angular Spring Constant(Angular Spring Constant=12000)
   > Set Is Not in Frontend(Is Not in Frontend:bool=true) -> (Output_Get:bool)
   > Branch(Condition=Get bIsPlayingEmote())
   > Set Angular Spring Constant R(Angular Spring Constant R=50000)
   > Set Angular Spring Constant(Angular Spring Constant=30000)
   n32 = GetAnimInstance(self=GetOwnerSkeletalMeshComponent())
   Is Valid(exec=@n1.then_2, InputObject=@n32)
   > Set IsTangyRadishEmote(IsTangyRadishEmote:bool=OR Boolean!(
     A=Montage_IsPlaying(self=@n32, Montage=/BRCosmetics/Animation/Game/MainPlayer/Emotes/TangyRadishMagma/Emote_TangyRadishMagma_CLF_M)
     B=Montage_IsPlaying(self=@n32, Montage=/BRCosmetics/Animation/Game/MainPlayer/Emotes/TangyRadishMagma/Emote_TangyRadishMagma_CLM_M)
     C:bool=Montage_IsPlaying(self=@n32, Montage=/BRCosmetics/Animation/Game/MainPlayer/Emotes/TangyRadishMagma/Emote_TangyRadishMagma_CMF_M)
     D:bool=Montage_IsPlaying(self=@n32, Montage=/BRCosmetics/Animation/Game/MainPlayer/Emotes/TangyRadishMagma/Emote_TangyRadishMagma_CMM_M)
    ) -> (ReturnValue)) -> (Output_Get:bool)
   Event OnBeginSkydiving()
   > Set DynamicsAlpha(DynamicsAlpha:double=1) -> (Output_Get:double)
   > Set SkydivingAlpha(SkydivingAlpha:Rotator="0,-20,0") -> (Output_Get:Rotator)
   Event OnEndSkydiving()
   > Set DynamicsAlpha(DynamicsAlpha=0.5)
   > Set SkydivingAlpha(SkydivingAlpha="0, 0, 0")
   TryGetPawnOwner()
   TryGetPawnOwner()
  ]}
]
variables [
 Angular Spring Constant {type=number}
 Angular Spring Constant R {type=number}
 BwdMovement {type=number}
 CrouchStrafe {type=boolean}
 DynamicsAlpha {type=number}
 ForwardGroundMovement {type=number}
 Gravity Override Params {type=GravityOverrideParamsStruct}
 Gravity Override Params_0 {
  type=FortGravityOverrideParameterSet
  default=(Skydive_GoingUp=(GravityOverride=()),Skydive_FloatingDown=(GravityOverride=()),Skydive_DivingDown=(GravityOverride=()),Skydive_Parachuting=(GravityOverride=()),OnGround_Standing=(GravityOverride=()),OnGround_Moving=(GravityOverride=()),OnGround_Moving_HighRotationRate=(GravityOverride=()),OnGround_Crouch_Moving=(GravityOverride=()),Emote_or_Melee=(GravityOverride=()),Swimming=(GravityOverride=()),Front_End=(GravityOverride=()),DBNO=(GravityOverride=()),DBNO_Carried=(GravityOverride=()),OnGround_Falling=(GravityOverride=()),Ziplining=(GravityOverride=()),Skydive_Surfing=(GravityOverride=()),Riding_Standing=(GravityOverride=()),Riding_Moving=(GravityOverride=()),Riding_Moving_Targeting=(GravityOverride=()),Tactical_Sprint=(GravityOverride=()),Rail_Grinding=(GravityOverride=()),)
 }
 GravityOverride {type=Vector}
 Is Crouching or Emoting {type=boolean}
 Is Not in Frontend {type=boolean}
 IsMovingStanding {type=boolean}
 IsTangyRadishEmote {type=boolean}
 Jumping RL {type=number}
 JumpingFalling {type=boolean}
 JumpingForwardAlpha {type=number}
 JumpingMovement {type=boolean}
 LeftMovement {type=number}
 PlayerVector_FwdBwd {type=number}
 PlayerVector_LR {type=number}
 RightMovement {type=number}
 SkydiveFloat movement {type=boolean}
 SkydivingAlpha {type=Rotator}
 StandingIdle {type=boolean}
 use_FE_float_logic {type=boolean}
]
